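# Evo state screen widget, derived from native EvoStateWidget, that layers a primary content stack, a bottom action bar and a flow debug label in an overlay.
# The inherited activatable stack fills the screen with zero-duration transitions, and the inherited bar sits bottom-right.
# Parent events DisplayDebugFlowText and HideDebugFlowText set and show, or collapse, the bordered debug label, which starts collapsed.
# PreConstruct gives that label placeholder text at design time only.
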WidgetBlueprint /Game/Evo/UI/Widgets/WBP_EvoStateWidget
graphs [EventGraph {nodes [
   n1 = Event PreConstruct()
   > Branch(Condition=@n1.IsDesignTime)
   > "SetText (Text)"(
    self=Get FlowProcessorDebug() -> (FlowProcessorDebug:Common Text Object Reference)
    InText="INVTEXT(\"DEBUG: Going through flow \\\"default\\\"\")"
   )
   n2 = Event DisplayDebugFlowText!() -> (OutputDelegate, then, DebugFlowText:"Text (by ref)") {is=title K2Node_Event}
   n3 = Get FlowProcessorDebug()
   > "SetText (Text)"(self=@n3, InText=@n2.DebugFlowText)
   > SetVisibility(self=@n3, InVisibility=HitTestInvisible)
   Event HideDebugFlowText() {is=title K2Node_Event}
   > SetVisibility(self=@n3, InVisibility=Collapsed)
  ]}]
variables []
widgets {
 info {inheritedWidgetCount=2, parentClass=/Script/FortniteUI.EvoStateWidget, rootWidgetClass=&Overlay, widgetCount=5}
 widgets [
  Overlay_30 {widgetClassPath=&Overlay, widget=Overlay_30}
  WidgetStack_PrimaryContent {
   widgetClassPath=/Script/CommonUI.CommonActivatableWidgetStack
   bInherited=true
   parent=Overlay_30
   widget=WidgetStack_PrimaryContent
   properties {
    navigation=&self:WidgetTree.WidgetStack_PrimaryContent.WidgetNavigation_0
    transitionCurveType=CubicInOut
    transitionDuration=0
   }
   slot {class=OverlaySlot, properties {horizontalAlignment=HAlign_Fill, verticalAlignment=VAlign_Fill}}
  }
  CommonBorder_175 {
   widgetClassPath=&CommonBorder
   parent=Overlay_30
   widget=CommonBorder_175
   properties {
    horizontalAlignment=HAlign_Left
    verticalAlignment=VAlign_Bottom
    background {
     bIsDynamicallyLoaded=false
     drawAs=Box
     imageType=NoImage
     mirroring=NoMirror
     resourceName=None
     resourceObject=null
     tiling=NoTile
     imageSize=<128 32>
     margin=<ltrb 0.1 0.7965 1 0.2611>
     outlineSettings {
      bUseBrushTransparency=false
      roundingType=HalfHeightRadius
      width=0
      color {colorUseRule=UseColor_Specified, specifiedColor=<rgba 0 0 0 0>}
      cornerRadii=<0 0 0 0>
     }
     tintColor {colorUseRule=UseColor_Specified, specifiedColor=<rgba 0.1667 0.1667 0.1667 1>}
     uVRegion {bIsValid=false, max=<0 0>, min=<0 0>}
    }
   }
   slot {class=OverlaySlot}
  }
  FlowProcessorDebug {
   widgetClassPath=&CommonTextBlock
   bIsVariable=true
   parent=CommonBorder_175
   widget=FlowProcessorDebug
   properties {
    visibility=Collapsed
    font {
     bForceMonospaced=false
     bMaterialIsStencil=false
     fontMaterial=null
     fontObject=&BurbankSmall
     letterSpacing=0
     monospacedWidth=1
     size=14
     skewAmount=0
     typefaceFontName=Bold
     outlineSettings {
      bApplyOutlineToDropShadows=false
      bMiteredCorners=false
      bSeparateFillAlpha=false
      outlineMaterial=null
      outlineSize=0
      outlineColor=<rgba 0 0 0 1>
     }
    }
    shadowOffset=<0 0>
   }
   slot {class=BorderSlot, properties {horizontalAlignment=HAlign_Left, verticalAlignment=VAlign_Bottom}}
  }
  BottomBar {
   widgetClassPath=/Game/UI/Frontend/BottomBar/AthenaBoundActionBar_BottomBar
   bInherited=true
   bIsVariable=true
   parent=Overlay_30
   widget=BottomBar
   slot {class=OverlaySlot, properties {horizontalAlignment=HAlign_Right, verticalAlignment=VAlign_Bottom, padding=<ltrb 0 0 80 32>}}
  }
 ]
}
